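# Translucent UI material that procedurally draws a hard-edged shape, named as the X for the Hired NPC elimination HUD graphic.
# Four rotated UV gradients (top, bottom and two end caps) each have an offset subtracted, then are steeply scaled and saturated into near-binary half-plane masks.
# Chained scalar DotProducts multiply the masks into their intersection, with rotation and offset parameters shaping each edge, so no texture is needed.
# A constant white color sits alongside the mask.
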Material /Game/Athena/HUD/Discoverability/materials/Hired_NPC_Elim/M_UI_NPCElim_X
graph {
 properties {blendMode=BLEND_Translucent, materialDomain=MD_UI, pixelDepthOffsetMode=PDOM_Legacy, usedShadingModels=Default Lit}
 nodes [
  Constant3Vector() {constant=<rgba 1 1 1 0>}
  n1 = ComponentMask(None=Rotator(Coordinate=TextureCoordinate(), Time=ScalarParameter() {defaultValue=28.2757, parameterName=1. Top Angle Rotation})) {g=true}
  n2 = ScalarParameter() {defaultValue=0.524, parameterName=1. Top Offset}
  n3 = Saturate(None=Multiply(A=Subtract(A=@n1, B=@n2), B=Constant() {r=128}))
  n4 = ComponentMask(None=Rotator(Coordinate=TextureCoordinate(), Time=ScalarParameter() {defaultValue=40.1057, parameterName=2. Bottom Angle Rotation})) {g=true}
  n5 = ScalarParameter() {defaultValue=0.3, parameterName=2. Bottom Offset}
  n6 = Saturate(None=Multiply(A=Subtract(A=@n4, B=@n5), B=Constant() {r=128}))
  n7 = Rotator(Coordinate=TextureCoordinate(), Time=ScalarParameter() {defaultValue=35.0003, parameterName=3. Bottom Cap Rotation})
  n8 = Multiply(
   A=Subtract(A=ComponentMask(None=@n7) {g=true}, B=ScalarParameter() {defaultValue=-0.092, parameterName=3. Bottom Cap offset})
   B=Constant() {r=128}
  )
  n9 = TextureCoordinate()
  n10 = ScalarParameter() {defaultValue=44.9737, parameterName=4. Top Cap Rotation}
  n11 = Subtract(A=ComponentMask(None=Rotator(Coordinate=@n9, Time=@n10)) {g=true}, B=ScalarParameter() {defaultValue=0.012, parameterName=4. Top Cap offset})
  n12 = Constant() {r=128}
  DotProduct(A=DotProduct(A=DotProduct(A=@n3, B=@n6), B=Saturate(None=@n8)), B=Saturate(None=Multiply(A=@n11, B=@n12)))
 ]
}
